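# One-shot mono MetaSound source for Mimosa companion mood-reaction voice, playing a random Mood.Sounds wave.
# A probability gate triggers playback with randomized pitch and volume, scaled by template multipliers, or finishes immediately.
# An affiliation compare routes unscaled audio to the locally-owned companion bus for envelope following.
# VoEnvelope outputs a mapped envelope follower times an allowVoSync-gated ADSR, whose release delays OnFinished when sync is on.
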MetaSoundSource /FortMetasound/Cosmetics/Mimosa/Mood/MSS_Mimosa_MoodReact
inputs [
 allowVoSync {type=Bool}
 Character.Affiliation {type=Enum:FortAudioAffiliation}
 Character.Health {type=Float}
 Character.UniqueID {type=Int32}
 Mood.Sounds {type=WaveAsset:Array}
 PitchModulationRange.Max {type=Float}
 PitchModulationRange.Min {type=Float}
 Probability {type=Float}
 TemplatePitchMultiplier {type=Float}
 TemplateVolumeMultiplier {type=Float}
 UE.Source.OnPlay {type=Trigger}
 VolumeModulationRange.Max {type=Float}
 VolumeModulationRange.Min {type=Float}
]
interfaces [Fort.Character:1.0, UE.OutputFormat.Mono:1.0, UE.Source.OneShot:1.0, UE.Source:1.1]
outputs [
 "UE.OutputFormat.Mono.Audio:0" {type=Audio}
 UE.Source.OneShot.OnFinished {type=Trigger}
 VoEnvelope {type=Float}
]
pages [Default {
  input_defaults {
   Character.Affiliation=0
   Character.Health=0
   Character.UniqueID=0
   PitchModulationRange.Max=1
   PitchModulationRange.Min=1
   Probability=1
   TemplatePitchMultiplier=1
   TemplateVolumeMultiplier=1
   UE.Source.OnPlay=null
   VolumeModulationRange.Max=1
   VolumeModulationRange.Min=1
   allowVoSync="<MetaSoundBuilderResult.SUCCEEDED: 0>"
   Mood.Sounds []
  }
  nodes [
   TriggerFilter_0 = UE.Trigger Filter(Probability=^Probability, Trigger=^UE.Source.OnPlay)
   ArrayRandomGet_0 = Array.Random Get:WaveAsset:Array(In Array=^Mood.Sounds, Next=@TriggerFilter_0.Heads) {also {defaults {Enable Shared State="<MetaSoundBuilderResult.SUCCEEDED: 0>"}}}
   n1 = UE.RandomFloat(Max=^PitchModulationRange.Max, Min=^PitchModulationRange.Min, Next=@ArrayRandomGet_0.On Next)
   FrequencyMultipliertoSemitone_0 = UE.Frequency Multiplier to Semitone:Float(Frequency Multiplier=UE.Multiply:Float(AdditionalOperands=^TemplatePitchMultiplier, PrimaryOperand=@n1.Value).Out)
   n2 = UE.RandomFloat(Max=^VolumeModulationRange.Max, Min=^VolumeModulationRange.Min, Next=@n1.On Next)
   WavePlayer_0 = UE.Wave Player:Mono(Wave Asset=@ArrayRandomGet_0.Value, Pitch Shift=@FrequencyMultipliertoSemitone_0.Semitones, Play=@n2.On Next) {also {version [1,1]}}
   n3 = &TriggerCompareAffiliation(
    A=^Character.Affiliation
    B=VariableAccessor.Enum:FortAudioAffiliation(Variable=InitVariable.Enum:FortAudioAffiliation().Variable).Value
    Compare=@WavePlayer_0.On Play
   )
   n4 = VariableAccessor.Audio(Variable=VariableMutator.Audio(Variable=InitVariable.Audio().Variable, Value=@WavePlayer_0.Out Mono).Variable)
   TriggerRouteTriggerRouteAudio2_0 = "TriggerRoute.Trigger Route (Audio, 2):Audio"(Set 1=UE.Reroute(Value=@n3.False).Value, Set 0=@n3.True, Value 0=VariableAccessor.Audio(Variable=@n4.Variable).Value)
   AudioBusWriter1_0 = "UE.Audio Bus Writer (1)"(In 0=@TriggerRouteTriggerRouteAudio2_0.Value) {also {defaults {Audio Bus=/Game/Sounds/Modulation/EnvelopeFollowers/Companion/AB_Companion_LocallyOwned}}}
   InitVariable.Audio()
   InitVariable.Enum:FortAudioAffiliation()
   InitVariable.Enum:FortAudioAffiliation()
   Multiply_0 = UE.Multiply:Audio by Float(PrimaryOperand=UE.Multiply:Audio by Float(
     AdditionalOperands=UE.Multiply:Float(AdditionalOperands=^TemplateVolumeMultiplier, PrimaryOperand=@n2.Value).Out
     PrimaryOperand=@WavePlayer_0.Out Mono
    ).Out)
   n5 = InitVariable.Trigger()
   n6 = TriggerCompare.Clamp:Bool(
    A=^allowVoSync
    Compare=VariableAccessor.Trigger(Variable=VariableMutator.Trigger(Variable=@n5.Variable, Value=@WavePlayer_0.On Finished).Variable).Value
   ) {also {defaults {B="<MetaSoundBuilderResult.SUCCEEDED: 0>"}}}
   n7 = VariableMutator.Trigger(Variable=InitVariable.Trigger().Variable, Value=@WavePlayer_0.On Play)
   ADSREnvelopeADSREnvelope_0 = ADSR Envelope.ADSR Envelope:Float(
    Trigger Release=@n6.True
    Trigger Attack=TriggerCompare.Clamp:Bool(A=^allowVoSync, Compare=VariableAccessor.Trigger(Variable=@n7.Variable).Value) {also {defaults {B="<MetaSoundBuilderResult.SUCCEEDED: 0>"}}}.True
   ) {also {defaults {Attack Time=0, Decay Time=0, Release Time=0.1, Sustain Level=1}}}
   EnvelopeFollower_0 = UE.Envelope Follower(In=@n4.Value) {also {defaults {Peak Mode=2}}}
   Multiply_3 = UE.Multiply:Float(
    PrimaryOperand=@ADSREnvelopeADSREnvelope_0.Out Envelope
    AdditionalOperands=MapRange.MapRange:Float(In=@EnvelopeFollower_0.Envelope) {also {defaults {In Range B=0.2}}}.Out Value
   )
   n8 = InitVariable.Trigger()
   TriggerAnyTriggerAny3_0 = "TriggerAny.Trigger Any (3)"(
    In 2=@ADSREnvelopeADSREnvelope_0.On Done
    In 0=@n6.False
    In 1=VariableAccessor.Trigger(Variable=VariableMutator.Trigger(Variable=@n8.Variable, Value=@TriggerFilter_0.Tails).Variable).Value
   )
  ]
  links [
   {dst="@UE.OutputFormat.Mono.Audio:0", src=Multiply_0.Out}
   {dst="@VoEnvelope", src=Multiply_3.Out}
   {dst="@UE.Source.OneShot.OnFinished", src=TriggerAnyTriggerAny3_0.Out}
  ]
 }]
properties {
 assetImportData=&self:AssetImportData
 attenuationSettings=/Game/Sounds/Attenuation/AS_Companion_VO
 audioPropertiesSheet=/CosmeticCompanions/Assets/Audio_Shared/PropertySheets/APS_Cosmetic_Mimosa_VO
 maxDistance=2200
 sampleRate=48000
 soundClassObject=&Fort_Creatures_Pets
 assetUserData [&self:AssetUserData_0]
 concurrencySet [/Game/Sounds/Concurrency_Rules/CS_Mimosa_VO, /Game/Sounds/Concurrency_Rules/CS_Mimosa_VO_Global]
}
